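# Layered Nanite-ready material for the MountainBase cave entrance mesh, built on material attributes with static switches for optional effects.
# Tiled concrete with a vertex-alpha secondary layer takes a world-height tint and UV2 overlay map.
# World-aligned rock detail height-blends by vertex red, and puddles by vertex blue on upward faces.
# A fixed-height waterline alters the surface, and counter-rotating noise adds fake caustics to base color on downward faces.
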Material /Game/Environments/Apollo/Sets/MountainBase/Materials/EntranceMesh/M_CaveEntrance
graph {
 properties {
  bUseMaterialAttributes=true
  bUsedWithInstancedStaticMeshes=true
  bUsedWithNanite=true
  pixelDepthOffsetMode=PDOM_Legacy
  usedShadingModels=Default Lit
 }
 nodes [
  n1 = Multiply(
   A=TextureCoordinate()
   B=AppendVector(
    A=ScalarParameter() {defaultValue=1, group=Textures, parameterName=Scale_X}
    B=ScalarParameter() {defaultValue=1, group=Textures, parameterName=Scale_Y}
   )
  )
  n2 = TextureSampleParameter2D(UVs=@n1) {group=Textures, parameterName=D, texture=/Game/Environments/Apollo/Sets/MountainBase/Textures/T_Concrete_Tile_D}
  n3 = Multiply(A=@n1, B=ScalarParameter() {defaultValue=2, group=Textures, parameterName=N2Tile, sortPriority=2})
  n4 = TextureSampleParameter2D(UVs=@n3) {group=Textures, parameterName=D2, sortPriority=3, texture=/Game/Environments/Apollo/Sets/PowerPlant/Textures/T_PP_Concrete_Wall_1_D}
  n5 = VertexColor()
  n6 = LinearInterpolate(B=@n4.RGB, Alpha=@n5.A) {constA=0.5, constB=0.5}
  n7 = TextureSampleParameter2D(UVs=@n1) {
   group=Textures
   parameterName=S
   samplerType=SAMPLERTYPE_Masks
   texture=/Game/Environments/Apollo/Sets/MountainBase/Textures/T_Concrete_Tile_S
  }
  n8 = TextureSampleParameter2D(UVs=@n1) {
   group=Textures
   parameterName=N
   samplerType=SAMPLERTYPE_Normal
   texture=/Game/Environments/Apollo/Sets/MountainBase/Textures/T_Concrete_Tile_N
  }
  n9 = Multiply(A=@n8.RGB, B=VectorParameter() {group=Textures, parameterName=NVal, sortPriority=1, defaultValue=<rgba 1 1 1 0>}.RGB)
  n10 = TextureSampleParameter2D(UVs=@n3) {
   group=Textures
   parameterName=N2
   samplerType=SAMPLERTYPE_Normal
   sortPriority=3
   texture=/Game/Environments/Apollo/Sets/PowerPlant/Textures/T_PP_CorrugatedMetal_Tile_N
  }
  n11 = VectorParameter() {group=Textures, parameterName=N2Val, sortPriority=4}
  n12 = Add(A=@n9, B=Multiply(A=Multiply(A=@n10.RGB, B=@n11.RGB), B=@n5.A))
  n13 = SetMaterialAttributes!(
   MaterialAttributes
   Base Color=StaticSwitchParameter(
    True=MaterialFunctionCall(Base=@n2.RGB, Blend=@n6) {materialFunction=&Blend_Overlay}.Result
    False=@n2.RGB
   ) {group=Textures, parameterName=N2True, sortPriority=1}
   Specular=@n7.R
   Roughness=@n7.R
   Normal=Normalize(VectorInput=StaticSwitchParameter(True=@n12, False=@n9) {group=Textures, parameterName=N2True, sortPriority=1})
   Metallic=@n7.R
  ) {attributeSetTypes [
    #1
    #2
    #3
    #4
    #5
   ]}
  n14 = GetMaterialAttributes(None=@n13) {attributeGetTypes [#1, #3, #2]}
  n15 = Add(A=ComponentMask(None=WorldPosition().XYZ) {b=true}, B=ScalarParameter() {defaultValue=3850, group=Tint, parameterName=TintStart})
  n16 = ScalarParameter() {defaultValue=1024, group=Tint, parameterName=TintBlur}
  n17 = LinearInterpolate(
   A=VectorParameter() {group=Tint, parameterName=TintBottom, defaultValue=<rgba 0.2969 0.2969 0.2969 0>}.RGB
   B=VectorParameter() {group=Tint, parameterName=TintTop, defaultValue=<rgba 1 1 1 0>}.RGB
   Alpha=Saturate(None=Divide(A=@n15, B=@n16))
  )
  n18 = VertexNormalWS()
  n19 = StaticSwitchParameter(
   True=LinearInterpolate(A=@n17, B=VectorParameter() {group=Tint, parameterName=TintUp, defaultValue=<rgba 1 1 1 0>}.RGB, Alpha=ComponentMask(None=@n18) {b=true})
   False=@n17
  ) {group=Tint, parameterName=TintUpTrue}
  n20 = TextureSampleParameter2D(UVs=TextureCoordinate() {coordinateIndex=1}) {group=Textures, parameterName="OverlayMap (UV2)", texture=/Game/Environments/Apollo/Sets/MountainBase/Textures/Entrance_LM}
  n21 = MaterialFunctionCall(In=@n20.R, Contrast=Constant() {r=1}) {materialFunction=&CheapContrast}
  n22 = SetMaterialAttributes!(
   MaterialAttributes=@n14.MaterialAttributes
   Base Color=Multiply(A=Multiply(A=@n14.Base Color, B=@n19), B=@n20.RGB)
   Roughness=Multiply(A=@n14.Roughness, B=@n21.Result)
   Specular=Multiply(A=@n14.Specular, B=@n21.Result)
  ) {attributeSetTypes [#1, #3, #2]}
  n23 = GetMaterialAttributes(None=@n22) {attributeGetTypes [
    #1
    #4
    #2
    #3
   ]}
  n24 = VectorParameter() {group=Details, parameterName=RockColor, defaultValue=<rgba 0.1354 0.1354 0.1354 0>}
  n25 = TextureObject() {texture=/Game/Environments/Apollo/Rocks/Natural_Rocks/Textures/T_RockDetails_02D}
  n26 = MaterialFunctionCall!(
   Diffuse Texture Object=@n25
   Spec Texture Object
   TextureSize=ScalarParameter() {defaultValue=1111, group=Details, parameterName=WSDetailTile, sortPriority=3}
   --------------------------
   Use seperate settings for Z projection
   Use seperate texture for Z projection
   Texture Size Z projection
   Z Plane Diffuse Texture Object
   Z Plane Spec Texture Object
   "--------------------------      "
   ProjectionTransitionContrast=Constant() {r=12}
   Use Texture Alpha Heightmap to Refine the Transistion
   HeightLerpTransitionContrast
   PickTheDiffuseHeightChannel
   TheDiffuseHeightChannel
   "--------------------------       "
   NormalTextureObject=TextureObject() {samplerType=SAMPLERTYPE_Normal, texture=/Game/Environments/Apollo/Rocks/Natural_Rocks/Textures/T_RockDetails_02_N}
   World Space Normals=StaticBool()
   "--------------------------  "
   Export Float 4 Diffuse
   Export Float 4 Spec
   "--------------------------     "
   WorldPosition
   World Space Normal
  ) {materialFunction=/Engine/Functions/Engine_MaterialFunctions02/Texturing/WorldAlignedTextures_Complex}
  n27 = Multiply(A=@n20.RGB, B=@n26.XYZ Diffuse Texture)
  n28 = ComponentMask(None=@n26.XYZ Diffuse Texture) {r=true}
  n29 = LinearInterpolate(A=ScalarParameter() {defaultValue=0.5, group=Details, parameterName=Details D Value}, Alpha=@n28)
  n30 = MaterialFunctionCall(Transition Phase=VertexColor().R, Height Texture=@n28, Contrast=Constant() {r=1}) {materialFunction=&HeightLerp}
  n31 = Multiply(A=@n26.Normal XYZ Texture, B=VectorParameter() {group=Details, parameterName=Details N Value, defaultValue=<rgba 0.2 0.2 0 1>}.RGB)
  n32 = Constant() {r=0.2762}
  n33 = Constant() {r=0.8}
  n34 = SetMaterialAttributes!(
   MaterialAttributes=@n23.MaterialAttributes
   Base Color=LinearInterpolate(A=Multiply(A=@n24.RGB, B=@n27), B=Multiply(A=@n23.Base Color, B=@n29), Alpha=@n30.Alpha)
   Normal=LinearInterpolate(A=@n26.Normal XYZ Texture, B=Add(A=@n23.Normal, B=@n31), Alpha=@n30.Alpha)
   Specular=LinearInterpolate(A=Multiply(A=@n28, B=@n32), B=@n23.Specular, Alpha=@n30.Alpha)
   Roughness=LinearInterpolate(A=Multiply(A=@n28, B=@n33), B=@n23.Roughness, Alpha=@n30.Alpha)
  ) {attributeSetTypes [
    #1
    #4
    #2
    #3
   ]}
  n35 = StaticSwitchParameter(True=@n34, False=@n22) {parameterName=DetailsBlendTrue}
  n36 = SetMaterialAttributes!(
   MaterialAttributes
   Base Color=Multiply(
    A=GetMaterialAttributes(None=@n35) {attributeGetTypes [#1]}.Base Color
    B=VectorParameter() {group=Puddles, parameterName=PuddlesColor, defaultValue=<rgba 0.5729 0.4582 0.3819 0>}.RGB
   )
   Specular=ScalarParameter() {defaultValue=0.1, group=Puddles, parameterName=PuddlesSpec}
   Roughness=ScalarParameter() {defaultValue=0.015, group=Puddles, parameterName=PuddlesRough}
   Normal=Constant3Vector() {constant=<rgba 0 0 1 0>}
  ) {attributeSetTypes [
    #1
    #2
    #3
    #4
   ]}
  n37 = Divide(
   A=ComponentMask(None=WorldPosition().XYZ) {g=true, r=true}
   B=ScalarParameter() {defaultValue=5555, group=Puddles, parameterName=PuddlesTexScale}
  )
  n38 = VertexNormalWS()
  n39 = Max(
   A=MaterialFunctionCall(
    Transition Phase=VertexColor().B
    Height Texture=TextureSample(UVs=@n37, Tex=@n25) {texture=&Entrance_LM}.R
    Contrast=ScalarParameter() {defaultValue=1, group=Puddles, parameterName=PuddlesContrast}
   ) {materialFunction=&HeightLerp}.Alpha
   B=OneMinus(None=ComponentMask(None=@n38) {b=true})
  )
  n40 = StaticSwitchParameter(True=BlendMaterialAttributes(A=@n36, B=@n35, Alpha=Saturate(None=@n39)), False=@n35) {parameterName=PuddlesTrue}
  n41 = GetMaterialAttributes(None=@n40) {attributeGetTypes [#1, #3, #2]}
  n42 = ComponentMask(None=WorldPosition().XYZ) {b=true}
  n43 = Constant() {r=-3900}
  n44 = Saturate(None=Divide(A=Subtract(A=@n42, B=@n43), B=ScalarParameter() {defaultValue=35, group=Waterline, parameterName=Waterline Height, sortPriority=3}))
  n45 = Multiply(
   A=@n41.Base Color
   B=LinearInterpolate(A=ScalarParameter() {defaultValue=0.5, group=Waterline, parameterName=Waterline Value}, Alpha=@n44) {constA=0.15}
  )
  n46 = LinearInterpolate(A=ScalarParameter() {defaultValue=0.2, group=Waterline, parameterName=Waterline Roughness, sortPriority=1}, B=@n41.Roughness, Alpha=@n44) {constA=0.5}
  n47 = LinearInterpolate(
   A=Multiply(A=@n41.Specular, B=ScalarParameter() {defaultValue=0.5, group=Waterline, parameterName=Waterline Spec, sortPriority=1})
   B=@n41.Specular
   Alpha=@n44
  )
  n48 = GetMaterialAttributes(None=StaticSwitchParameter(
    True=SetMaterialAttributes!(MaterialAttributes=@n41.MaterialAttributes, Base Color=@n45, Roughness=@n46, Specular=@n47) {attributeSetTypes [#1, #3, #2]}
    False=@n40
   ) {parameterName=WaterlineTrue}) {attributeGetTypes [#1]}
  n49 = DotProduct(A=VertexNormalWS(), B=Constant3Vector() {constant=<rgba 0 0 -1 0>})
  n50 = ComponentMask(None=Subtract(A=ActorPositionWS(), B=WorldPosition().XYZ)) {r=true}
  n51 = ScalarParameter() {defaultValue=768, group=Caustics, parameterName=Side Fade Blr}
  n52 = OneMinus(None=Abs(None=Divide(A=@n50, B=@n51)))
  n53 = Subtract(A=WorldPosition().XYZ, B=Constant3Vector() {constant=<rgba -16896 112896 0 0>})
  n54 = ComponentMask(None=@n53) {g=true, r=true}
  n55 = ScalarParameter() {defaultValue=737.9165, group=Caustics, parameterName=CausticsScale}
  n56 = TextureSampleParameter2D(UVs=Rotator(Coordinate=Divide(A=@n54, B=@n55)) {speed=0.0125}) {group=Caustics, parameterName=Lighting, samplerType=SAMPLERTYPE_Grayscale, texture=&LowResBlurredNoise}
  n57 = ComponentMask(None=Add(A=@n53, B=Constant3Vector() {constant=<rgba 555 5555 0 0>})) {g=true, r=true}
  n58 = TextureSampleParameter2D(UVs=Rotator(Coordinate=Divide(A=@n57, B=@n55)) {speed=-0.0075}) {group=Caustics, parameterName=Lighting, samplerType=SAMPLERTYPE_Grayscale, texture=&LowResBlurredNoise}
  n59 = Multiply(
   A=Multiply(A=Saturate(None=@n49), B=Saturate(None=@n52))
   B=Multiply(A=Multiply(A=@n56.R, B=@n58.R), B=ScalarParameter() {defaultValue=1, group=Caustics, parameterName=CausticValue})
  )
  StaticSwitchParameter(
   True=SetMaterialAttributes!(MaterialAttributes=@n48.MaterialAttributes, Base Color=Add(A=@n48.Base Color, B=@n59)) {attributeSetTypes [#1]}
   False=@n48.MaterialAttributes
  ) {parameterName=CausticTrue}
 ]
}
